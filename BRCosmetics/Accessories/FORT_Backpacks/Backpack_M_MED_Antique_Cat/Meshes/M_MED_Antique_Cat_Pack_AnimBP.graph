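# Animation Blueprint for the Antique Cat backpack cosmetic, adding posture-driven tilt and physics motion to the pack.
# Each update maps the pitch of spine_05 relative to spine_01 into a clamped roll, applied additively to the root bone.
# An additional fixed roll on the pack bone is active only while skydiving.
# A RigidBody node then simulates the pack's physics asset in pack-bone space with damped component motion and overridden gravity.
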AnimBlueprint /BRCosmetics/Accessories/FORT_Backpacks/Backpack_M_MED_Antique_Cat/Meshes/M_MED_Antique_Cat_Pack_AnimBP
target_skeleton=&/M_MED_Antique_Cat_Pack_Skeleton
graphs [
 AnimGraph {nodes [
   n1 = Local To Component(LocalPose=Input Pose())
   n2 = Get backpackRotate() -> (backpackRotate:Rotator)
   n3 = Get bIsSkyDiving()
   n4 = RigidBody(ComponentPose="Transform (Modify) Bone - Bone: pack"(
     ComponentPose="Transform (Modify) Bone - Bone: root"(ComponentPose=@n1, Translation="0,0,0", Rotation=@n2) {rotationMode=BMM_Additive, alphaBoolBlend {blendOption=Linear}, boneToModify {boneName=root}, rotation {roll=9}}
     Alpha="To Float (Boolean)"(InBool=@n3)
     Translation="0,0,0"
     Rotation="0,0,15"
    ) {rotationMode=BMM_Additive, alphaBoolBlend {blendOption=Linear}, boneToModify {boneName=pack}, rotation {roll=15}}) {
    bClampLinearTranslationLimitToRefPose=true
    bOverrideWorldGravity=true
    overridePhysicsAsset=&/M_MED_Antique_Cat_Pack_Physics
    simulationSpace=BaseBoneSpace
    alphaBoolBlend {blendOption=Linear}
    baseBoneRef {boneName=pack}
    componentAppliedLinearAccClamp=<900 900 900>
    componentLinearAccScale=<0.3 0.3 0.3>
    componentLinearVelScale=<0.35 0.35 0.35>
    overrideWorldGravity {z=-980}
   }
   Output Pose(Result=Component To Local(ComponentPose=@n4)) {layerGroup=None}
  ]}
 EventGraph {nodes [
   Event BlueprintUpdateAnimation()
   n1 = GetOwnerSkeletalMeshComponent()
   n2 = BreakRotator(InRot=BreakTransform(InTransform=GetDeltaTransformFromRefPose(self=@n1, BoneName=spine_05, BaseName=spine_01)).Rotation)
   > Set backpackRotate(backpackRotate:Rotator=MakeRotator(Roll=MapRangeClamped(Value=@n2.Pitch, InRangeA=-25, InRangeB=20, OutRangeB=10))) -> (Output_Get:Rotator)
   TryGetPawnOwner()
  ]}
]
variables [backpackRotate {type=Rotator}]
